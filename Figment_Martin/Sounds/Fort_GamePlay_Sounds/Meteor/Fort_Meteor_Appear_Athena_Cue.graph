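# Sound cue for a meteor appearing, going by its name, playing a sonic boom.
# A Random node picks between two equally weighted boom waves, feeding a Modulator that varies pitch slightly on each play.
# The cue plays quietly through an important in-game mission sound class, uses meteor alert attenuation, and overrides concurrency to cap simultaneous instances.
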SoundCue /Figment_Martin/Sounds/Fort_GamePlay_Sounds/Meteor/Fort_Meteor_Appear_Athena_Cue
root=Modulator_0
nodes [
 {id=Modulator_0, children [Random_0], properties {pitchMax=1.2, pitchMin=0.9}}
 {id=Random_0, children [WavePlayer_1,WavePlayer_2], properties {weights [1,1]}}
 {id=WavePlayer_1, properties {soundWaveAssetPtr=&/Meteor_Sonic_Boom_01}}
 {id=WavePlayer_2, properties {soundWaveAssetPtr=&/Meteor_Sonic_Boom_02}}
]
properties {
 attenuationSettings=&/Fort_Meteor_Alert_Athena_Att
 bOverrideConcurrency=true
 maxDistance=25000
 soundClassObject=&Fort_Mission_Ingame_Important_PSM
 volumeMultiplier=0.1
 concurrencyOverrides {maxCount=2}
}
