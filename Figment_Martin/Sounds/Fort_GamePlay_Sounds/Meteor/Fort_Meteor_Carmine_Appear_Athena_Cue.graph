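# Sound cue for the Carmine meteor's appearance sound in an Athena game mode.
# It picks one of two appear waves at equal weight and passes it through a modulator that randomizes pitch.
# It plays at reduced volume on the important in-game mission sound class, using the Carmine alert attenuation asset with a very long max distance.
# A concurrency override limits it to two simultaneous instances.
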SoundCue /Figment_Martin/Sounds/Fort_GamePlay_Sounds/Meteor/Fort_Meteor_Carmine_Appear_Athena_Cue
root=Modulator_0
nodes [
 {id=Modulator_0, children [Random_1], properties {pitchMax=1.2, pitchMin=0.9}}
 {id=Random_1, children [WavePlayer_0,WavePlayer_3], properties {weights [1,1]}}
 {id=WavePlayer_0, properties {soundWaveAssetPtr=&/Meteor_Carmine_Appear_01}}
 {id=WavePlayer_3, properties {soundWaveAssetPtr=&/Meteor_Carmine_Appear_02}}
]
properties {
 attenuationSettings=&/Fort_Meteor_Carmine_Alert_Athena_Att
 bOverrideConcurrency=true
 maxDistance=300000
 soundClassObject=&Fort_Mission_Ingame_Important_PSM
 volumeMultiplier=0.5
 concurrencyOverrides {maxCount=2}
}
