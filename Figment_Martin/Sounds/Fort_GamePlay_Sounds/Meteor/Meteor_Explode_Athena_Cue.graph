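# Sound cue for a meteor explosion in Athena, layering a close break-open sound with a distant explosion.
# A mixer combines both layers, each randomly picking one of two waves, then passing through a modulator and its own close or distant projectile attenuation node.
# The cue plays in an impact sound class with slightly reduced volume and a concurrency override capping simultaneous instances.
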SoundCue /Figment_Martin/Sounds/Fort_GamePlay_Sounds/Meteor/Meteor_Explode_Athena_Cue
root=Mixer_0
nodes [
 {id=Attenuation_0, children [Modulator_2], properties {attenuationSettings=&Fort_Weapon_Projectile_Close_Athena_Att}}
 {
  id=Attenuation_1
  children [Modulator_0]
  properties {attenuationSettings=&Fort_Weapon_Projectile_Distant_Mono_Athena_Att}
 }
 {id=Mixer_0, children [Attenuation_0, Attenuation_1], properties {inputVolume [1,1]}}
 {id=Modulator_0, children [Random_1]}
 {id=Modulator_2, children [Random_0]}
 {id=Random_0, children [WavePlayer_1,WavePlayer_2], properties {weights [1,1]}}
 {id=Random_1, children [WavePlayer_0,WavePlayer_4], properties {weights [1,1]}}
 {id=WavePlayer_0, properties {soundWaveAssetPtr=/Game/Sounds/Fort_GamePlay_Sounds/Meteor/Meteor_Distant_Explosion_01}}
 {id=WavePlayer_1, properties {soundWaveAssetPtr=/Game/Sounds/Fort_GamePlay_Sounds/Meteor/Meteor_BreakOpen_02}}
 {id=WavePlayer_2, properties {soundWaveAssetPtr=/Game/Sounds/Fort_GamePlay_Sounds/Meteor/Meteor_BreakOpen_01}}
 {id=WavePlayer_4, properties {soundWaveAssetPtr=/Game/Sounds/Fort_GamePlay_Sounds/Meteor/Meteor_Distant_Explosion_02}}
]
properties {
 attenuationSettings=/Game/Sounds/Attenuation/Fort_Prop_Close_Att
 bOverrideConcurrency=true
 maxDistance=27500
 soundClassObject=/Game/Sounds/FortSoundClassesAndModes/Fort_ImpactSounds_DuckDefault_PSM
 volumeMultiplier=0.8
 concurrencyOverrides {maxCount=4}
}
