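# Sound cue for a meteor sub-explosion in Athena, mixing a main impact layer with a distant explosion layer on the ambient sound class.
# The main layer randomly picks one of three sub-impact waves through a modulator and meteor attenuation.
# The distant layer randomly picks one of two distant explosion waves, pitched up by a modulator, with its own distant attenuation.
# A concurrency override caps simultaneous instances, and overall volume is boosted.
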SoundCue /Figment_Martin/Sounds/Fort_GamePlay_Sounds/Meteor/Meteor_SubExplode_Athena_Cue
root=Mixer_0
nodes [
 {
  id=Attenuation_4
  children [Modulator_2]
  properties {attenuationSettings=&/Fort_Meteor_Athena_Att}
 }
 {
  id=Attenuation_5
  children [Modulator_0]
  properties {attenuationSettings=&/Fort_Meteor_Distant_Athena_Att}
 }
 {id=Mixer_0, children [Attenuation_4, Attenuation_5], properties {inputVolume [1,1]}}
 {id=Modulator_0, children [Random_2], properties {pitchMax=1.5, pitchMin=1.3}}
 {id=Modulator_2, children [Random_1]}
 {id=Random_1, children [WavePlayer_0,WavePlayer_4,WavePlayer_5], properties {weights [1,1,1]}}
 {id=Random_2, children [WavePlayer_6,WavePlayer_7], properties {weights [1,1]}}
 {id=WavePlayer_0, properties {soundWaveAssetPtr=&/Meteor_SubImpact_01}}
 {id=WavePlayer_4, properties {soundWaveAssetPtr=&/Meteor_SubImpact_02}}
 {id=WavePlayer_5, properties {soundWaveAssetPtr=&/Meteor_SubImpact_03}}
 {id=WavePlayer_6, properties {soundWaveAssetPtr=/Game/Sounds/Fort_GamePlay_Sounds/Meteor/Meteor_Distant_Explosion_01}}
 {id=WavePlayer_7, properties {soundWaveAssetPtr=/Game/Sounds/Fort_GamePlay_Sounds/Meteor/Meteor_Distant_Explosion_02}}
]
properties {
 bOverrideConcurrency=true
 maxDistance=22000
 soundClassObject=&Fort_Ambient_Athena
 volumeMultiplier=2.0741
 concurrencyOverrides {maxCount=4}
}
